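; Gameplay ability for the Injured Prey medallion that reveals nearby player pawns with low combined health and shield.
; On activation, a looping one-second timer scans pawns around the owner and collects others at or below the effective-HP threshold.
; Those targets join the owner controller's indicated-actor list through filter data.
; Pawns not already indicated get a burst enemy GameplayCue, and any new find fires one cue on the owner.
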(event Ability|EventActivateAbility
  (bind _outfortplayerpawn (CallFunction|GetAvatarasValidFortPlayerPawn)
    (:ValidFortPlayerPawn
      (bind _output_get (Variables|Default|SetPlayer%20Pawn _outfortplayerpawn))
      (Variables|Default|SetFortPlayerController (Fortnite|GetFortPlayerControllerfromActor _output_get))
      (Utilities|Time|SetTimerbyEvent (delegate Custom|Check%20Enemy%20Health) 1.0 true))))

(event Custom|Check%20Enemy%20Health
  (Utilities|Array|Clear (Variables|Default|GetTargetedPawns))
  (Variables|Default|SetFoundNewPawn false)
  (bind _outfortplayerpawns (Fortnite|GetFortPlayerPawnsInRange (Transformation|GetActorLocation (Variables|Default|GetPlayer%20Pawn)) 1e+04 true))
  (for _array_element _outfortplayerpawns
    (switch Utilities|FlowControl|Switch|SwitchonEFortTeamAffiliation (Team|GetTeamAffiliationforActortoLocalPlayer _array_element)
      (:Neutral
        _)
      (:Hostile
        _))
    (if (and (<= (+ (FortPawn|GetShield _array_element) (FortPawn|GetHealth _array_element)) (Variables|Default|GetEHPThreshold)) (!= _array_element (Variables|Default|GetPlayer%20Pawn)))
      (Utilities|Array|Add (Variables|Default|GetTargetedPawns) _array_element)
      (if (not (IndicatedManagement|IsActorinIndicatedListforController (Variables|Default|GetFortPlayerController) _array_element))
        (GameplayCue|ExecuteGameplayCueOnActor(Burst) _array_element (Variables|Default|GetEnemyCue) (Ability|GameplayCue|MakeGameplayCueParameters 1.0))
        (Variables|Default|SetFoundNewPawn true))))
  (IndicatedManagement|AddActorsToIndicatedListWithFilter (Fortnite|GetFortPlayerControllerfromActor (Variables|Default|GetPlayer%20Pawn)) (Variables|Default|GetTargetedPawns) (Variables|Default|GetIndicated%20Actor%20Filter%20Datas) false false true 0 false)
  (if (Variables|Default|GetFoundNewPawn)
    (Ability|ExecuteGameplayCueWithParamsOnOwner (Variables|Default|GetSelfCue) (Ability|GameplayCue|MakeGameplayCueParameters 1.0))))
